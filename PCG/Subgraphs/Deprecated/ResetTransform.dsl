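; A deprecated PCG subgraph that resets the transform of its input points to a default value.
; It adds a temporary Transform attribute, tmpTransform, left at its default, and copies it onto each point's $Transform property.
; It then deletes the temporary attribute and passes the result to the output.
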(graph ResetTransform
  (bind CreateAttribute_0 (Add%20Attribute :AttributeTypes (struct :Type Transform) :InputSource @LastCreated :OutputTarget tmpTransform))
  (bind AttributeOperation_1 (Copy%20Attributes :InputSource tmpTransform :OutputTarget $Transform))
  (bind FilterAttribute_2 (Delete%20Attributes :bTokenizeOnWhiteSpace true :SelectedAttributes "tmpTransform"))
  (connect CreateAttribute_0 Out AttributeOperation_1 Source)
  (connect CreateAttribute_0 Out AttributeOperation_1 Target)
  (connect DefaultInputNode In CreateAttribute_0 In)
  (connect AttributeOperation_1 Out FilterAttribute_2 In)
  (output Out FilterAttribute_2))
